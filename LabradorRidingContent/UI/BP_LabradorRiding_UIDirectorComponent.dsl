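; UI director component that shows the Labrador info UI scene while the local player's view target is riding a Labrador.
; When the UI is ready, it binds to the player controller's view target changes and evaluates the current target, unbinding on end play.
; For each valid view target, async gameplay tag tasks watch the rider tag being added or removed.
; A delta gate adds or removes the scene only when the active state changes.
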(event EventOnReadyToShowUI
  (bind _asfort_player_controller (Utilities|Casting|CastToFortPlayerController (DynamicUIDirector|GetOwningLocalPlayerController (CallFunction|GetOwningFortDynamicUIDirector))))
  (EventDispatchers|BindEventtoOnViewTargetChanged _asfort_player_controller (EventDispatchers|CreateEvent :!event Custom|OnViewTargetChanged))
  (CallFunction|OnViewTargetChanged 0 0 (Pawn|GetViewTarget _asfort_player_controller)))

(event Custom|OnViewTargetChanged
  (params (PlayerController :type Fort%20Player%20Controller%20Object%20Reference) (OldViewTarget :type Actor%20Object%20Reference) (NewViewTarget :type Actor%20Object%20Reference))
  (Utilities|IsValid NewViewTarget
    (:Is%20Valid
      ; Add scene when player is riding
      (Ability|Tasks|WaitGameplayTagAddToActor NewViewTarget (Variables|Default|GetLabradorRider_GameplayTag)
        (:then
          ; Remove scene if player is not riding
          (Ability|Tasks|WaitGameplayTagRemoveFromActor NewViewTarget (Variables|Default|GetLabradorRider_GameplayTag)
            (:Removed
              (Default|AddLabradorInfoUIScene false))))
        (:Added
          (Default|AddLabradorInfoUIScene true))))))

(event EventEndPlay
  (params (EndPlayReason :type EEndPlayReason%20Enum :default "Destroyed"))
  (bind _asfort_player_controller (Utilities|Casting|CastToFortPlayerController (DynamicUIDirector|GetOwningLocalPlayerController (CallFunction|GetOwningFortDynamicUIDirector))))
  (EventDispatchers|UnbindEventfromOnViewTargetChanged _asfort_player_controller (EventDispatchers|CreateEvent :!event Custom|OnViewTargetChanged)))

(fn AddLabradorInfoUIScene
  (params (SetActive :type Boolean))
  (bind _returnvalue (CallFunction|GetOwningFortDynamicUIDirector))
  (bind _labradorinfo_uiscene (Variables|Default|GetLabradorInfo_UIScene))
  (bind _updated_variable (Utilities|FlowControl|DeltaGateWildcard (Variables|Default|GetLabradorInfoUISceneActive) SetActive)
    (:Success
      (if _updated_variable
        (DynamicUIDirector|AddScene _returnvalue _labradorinfo_uiscene)
        (else
          (DynamicUIDirector|RemoveScene _returnvalue _labradorinfo_uiscene))))))
