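# NPC status widget extension that shows a small icon on a rideable Labrador creature when its rider shares the local player's squad.
# It starts collapsed, finds the ridable component on the assigned FortPawn, checks any current rider and binds to RidableUIStart for later mounts.
# A squad ID comparison either shows the icon with the component's IconBrush, ignoring the event's Icon parameter, or collapses the widget.
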WidgetBlueprint /LabradorRidingContent/UI/LabradorRiding_RiderIcon
graphs [
 EventGraph {nodes [
   n1 = RidableStart!() -> (OutputDelegate, then, Name:Text, Icon:Slate Brush Structure) {is=title K2Node_CustomEvent}
   n2 = IsRiderOnSameSquad!(execute=[@n1.then, @n9.Is Valid], self) -> (then, SameSquad:bool)
   > n3 = Branch(Condition=@n2)
   > SetVisibility(InVisibility=HitTestInvisible)
   > UpdateIcon!(
    execute
    self
    InBrush:"Slate Brush Structure (by ref)"=Get IconBrush!(self:Labrador Riding Creature Base Ridable Component Object Reference=Get Ridable Base Component() -> (Ridable Base Component:Labrador Riding Creature Base Ridable Component Object Reference)) -> (IconBrush:Slate Brush Structure)
   ) -> (then)
   SetVisibility(execute=@n3.else, InVisibility=Collapsed)
   n4 = Event SetFortPawn!() -> (OutputDelegate, then, FortPawn:Fort Pawn Object Reference) {is=title K2Node_Event}
   > n5 = Set Fort Pawn(Fort Pawn:Fort Pawn Object Reference=@n4.FortPawn) -> (Output_Get:Fort Pawn Object Reference)
   n6 = Get Fort Pawn() -> (Fort Pawn:Fort Pawn Object Reference)
   Is Valid(exec=[@n10.then, @n5.then], InputObject=@n6)
   n7 = GetComponentByClass!(
    self:Actor Object Reference=@n6
    ComponentClass:Actor Component Class Reference=/LabradorRidingContent/Components/LabradorRiding_CreatureBaseRidableComponent.LabradorRiding_CreatureBaseRidableComponent_C
   ) -> (ReturnValue:Labrador Riding Creature Base Ridable Component Object Reference)
   > Is Valid(InputObject=@n7)
   > Set Ridable Base Component(Ridable Base Component:Labrador Riding Creature Base Ridable Component Object Reference=@n7) -> (Output_Get:Labrador Riding Creature Base Ridable Component Object Reference)
   > n8 = Sequence!(execute) -> (then_0, then_1, then_2:Exec)
   > n9 = Is Valid(InputObject=Get PlayerRiderComponentRef!(self:Labrador Riding Creature Base Ridable Component Object Reference=Get Ridable Base Component()) -> (PlayerRiderComponentRef:Labrador Riding Player Rider Component Object Reference))
   Bind Event to Ridable UIStart(execute=@n8.then_1, self:Labrador Riding Creature Base Ridable Component Object Reference=Get Ridable Base Component(), Delegate=@n1) {is=title K2Node_AddDelegate}
   Event Construct()
   > n10 = SetVisibility(InVisibility=Collapsed)
  ]}
 IsRiderOnSameSquad {nodes [
   IsRiderOnSameSquad() {is=title K2Node_FunctionEntry}
   n1 = Get Ridable Base Component()
   > n2 = Is Valid(InputObject=@n1)
   Return Node!(execute=[@n5.CastFailed, @n8.CastFailed, @n9.CastFailed, @n2.Is Not Valid, @n4.Is Not Valid, @n7.Is Not Valid], SameSquad:bool=false)
   n3 = GetOwner(self=Get PlayerRiderComponentRef(self=@n1))
   > n4 = Is Valid(InputObject=@n3)
   > n5 = Cast To FortPlayerPawnAthena(Object=@n3)
   n6 = Get PlayerState!(self:Pawn Object Reference=@n5) -> (PlayerState:Player State Object Reference)
   > n7 = Is Valid(InputObject=@n6)
   > n8 = Cast To FortPlayerStateAthena!(execute, Object=@n6) -> (then, CastFailed, AsFort Player State Athena:Fort Player State Athena Object Reference) {is=title K2Node_DynamicCast}
   > n9 = Cast To FortPlayerStateAthena(Object=Get PlayerState(self=GetOwningPlayerPawn!(self:User Widget Object Reference) -> (ReturnValue:Pawn Object Reference)))
   > Return Node(SameSquad="Equal (Byte)"(
     A:Byte=GetSquadId!(self:Fort Player State Athena Object Reference=@n9) -> (ReturnValue:Byte=0)
     B:Byte=GetSquadId(self=@n8) -> (ReturnValue=0)
    ) -> (ReturnValue:bool=~) {is=title K2Node_PromotableOperator})
  ]}
 UpdateIcon {nodes [
   n1 = UpdateIcon!() -> (then, InBrush:"Slate Brush Structure (by ref)") {is=title K2Node_FunctionEntry}
   > SetVisibility(InVisibility=HitTestInvisible)
   > SetBrush(self=Get SprintIcon() -> (SprintIcon:Image Object Reference), InBrush=@n1)
  ]}
]
variables [
 Fort Pawn {type=FortPawn, default=null}
 Ridable Base Component {type=LabradorRiding_CreatureBaseRidableComponent_C, default=null}
]
widgets {
 info {inheritedWidgetCount=0, parentClass=/Script/FortniteUI.NPCStatusWidgetExtensionBase, rootWidgetClass=&Overlay, widgetCount=3}
 widgets [
  Overlay_Stats {widgetClassPath=&Overlay, bIsVariable=true, widget=Overlay_Stats}
  SizeBox_1 {
   widgetClassPath=&SizeBox
   parent=Overlay_Stats
   widget=SizeBox_1
   properties {bOverride_HeightOverride=true, bOverride_WidthOverride=true, heightOverride=32, widthOverride=32}
   slot {class=OverlaySlot, properties {horizontalAlignment=HAlign_Fill, verticalAlignment=VAlign_Fill}}
  }
  SprintIcon {
   widgetClassPath=&Image
   bIsVariable=true
   parent=SizeBox_1
   widget=SprintIcon
   properties {brush {
     bIsDynamicallyLoaded=false
     drawAs=Image
     imageType=NoImage
     mirroring=NoMirror
     resourceName=None
     resourceObject=/Game/Athena/UI/InGame/Creative/PropertyEditor/Icon_Picker/Textures/128x128/T_UI_IconLibrary_exclamation_128
     tiling=NoTile
     imageSize=<32 32>
     margin=<ltrb 0.2 0.2 0.2 0.2>
     outlineSettings {
      bUseBrushTransparency=false
      roundingType=HalfHeightRadius
      width=0
      color {colorUseRule=UseColor_Specified, specifiedColor=<rgba 0 0 0 0>}
      cornerRadii=<0 0 0 0>
     }
     tintColor {colorUseRule=UseColor_Specified, specifiedColor=<rgba 1 1 1 1>}
     uVRegion {bIsValid=false, max=<0 0>, min=<0 0>}
    }}
   slot {class=SizeBoxSlot, properties {horizontalAlignment=HAlign_Center, verticalAlignment=VAlign_Center}}
  }
 ]
}
