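# NPC status widget extension that shows a ridden creature's stamina as a thin bar to players in the rider's squad.
# It starts collapsed, finds the pawn's ridable component, then on ride start or an existing rider checks squad IDs to show or collapse itself.
# When shown, it hides the wildlife NPC's status and binds energy updates.
# Updates drive fill and delta material Progress, with red tint and looping opacity pulse while resting.
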WidgetBlueprint /LabradorRidingContent/UI/LabradorRiding_RiderStaminaBar
animations [RechargingAnim {
  binding_count=1
  length_seconds=0.5
  object=&self:RechargingAnim
  sequence {bindings [Overlay_Stats {
     id=#1
     possessed_class=Overlay
     tracks [{
       type=MovieSceneFloatTrack
       display_name=Render Opacity
       sections [{
         type=MovieSceneFloatSection
         end_frame=10
         start_frame=0
         channels [None {type=MovieSceneScriptingFloatChannel, keys [[0,1],[5,0.33],[10,1]]}]
        }]
      }]
    }]}
 }]
graphs [
 EventGraph {nodes [
   n1 = RidableStart!() -> (OutputDelegate, then, Name:Text, Icon:Slate Brush Structure) {is=title K2Node_CustomEvent}
   n2 = IsRiderOnSameSquad!(execute=[@n1.then, @n12.Is Valid], self) -> (then, SameSquad:bool)
   > n3 = Branch(Condition=@n2)
   > SetVisibility(InVisibility=HitTestInvisible)
   > n4 = UpdateEnergy!(
    execute
    self
    Value:double=Get currentEnergyValue!(self:Labrador Riding Creature Base Ridable Component Object Reference=Get RidableBaseComponent() -> (RidableBaseComponent:Labrador Riding Creature Base Ridable Component Object Reference)) -> (currentEnergyValue:double)
    bIsResting:bool=false
   ) -> (then)
   n5 = RidableEnergy!() -> (OutputDelegate, then, Energy:double=~, bIsResting:bool=~) {is=title K2Node_CustomEvent}
   Bind Event to Ridable UIUpdate Energy(execute=@n4.then, self:Labrador Riding Creature Base Ridable Component Object Reference=Get RidableBaseComponent(), Delegate=@n5) {is=title K2Node_AddDelegate}
   > n6 = Cast To NPC_Pawn_Wildlife_Parent!(execute, Object=Get FortPawn() -> (FortPawn:Fort Pawn Object Reference)) -> (then, CastFailed, AsNPC Pawn Wildlife Parent:NPC Pawn Wildlife Parent Object Reference) {is=title K2Node_DynamicCast}
   > HideStatus(self:NPC Pawn Wildlife Parent Object Reference=@n6)
   SetVisibility(execute=@n3.else, InVisibility=Collapsed)
   n7 = Event SetFortPawn!() -> (OutputDelegate, then, FortPawn:Fort Pawn Object Reference) {is=title K2Node_Event}
   > n8 = Set FortPawn(FortPawn:Fort Pawn Object Reference=@n7.FortPawn) -> (Output_Get:Fort Pawn Object Reference)
   n9 = Get FortPawn()
   Is Valid(exec=[@n13.then, @n8.then], InputObject=@n9)
   n10 = GetComponentByClass!(
    self:Actor Object Reference=@n9
    ComponentClass:Actor Component Class Reference=/LabradorRidingContent/Components/LabradorRiding_CreatureBaseRidableComponent.LabradorRiding_CreatureBaseRidableComponent_C
   ) -> (ReturnValue:Labrador Riding Creature Base Ridable Component Object Reference)
   > Is Valid(InputObject=@n10)
   > Set RidableBaseComponent(RidableBaseComponent:Labrador Riding Creature Base Ridable Component Object Reference=@n10) -> (Output_Get:Labrador Riding Creature Base Ridable Component Object Reference)
   > n11 = Sequence!(execute) -> (then_0, then_1, then_2:Exec)
   > n12 = Is Valid(InputObject=Get PlayerRiderComponentRef!(self:Labrador Riding Creature Base Ridable Component Object Reference=Get RidableBaseComponent()) -> (PlayerRiderComponentRef:Labrador Riding Player Rider Component Object Reference))
   Bind Event to Ridable UIStart(execute=@n11.then_1, self:Labrador Riding Creature Base Ridable Component Object Reference=Get RidableBaseComponent(), Delegate=@n1) {is=title K2Node_AddDelegate}
   Event Construct()
   > n13 = SetVisibility(InVisibility=Collapsed)
   UpdateEnergy(execute=@n5.then, Value=@n5.Energy, bIsResting=@n5.bIsResting)
  ]}
 IsRiderOnSameSquad {nodes [
   IsRiderOnSameSquad() {is=title K2Node_FunctionEntry}
   n1 = Get RidableBaseComponent()
   > n2 = Is Valid(InputObject=@n1)
   Return Node!(execute=[@n5.CastFailed, @n8.CastFailed, @n9.CastFailed, @n2.Is Not Valid, @n4.Is Not Valid, @n7.Is Not Valid], SameSquad:bool=false)
   n3 = GetOwner(self=Get PlayerRiderComponentRef(self=@n1))
   > n4 = Is Valid(InputObject=@n3)
   > n5 = Cast To FortPlayerPawnAthena(Object=@n3)
   n6 = Get PlayerState!(self:Pawn Object Reference=@n5) -> (PlayerState:Player State Object Reference)
   > n7 = Is Valid(InputObject=@n6)
   > n8 = Cast To FortPlayerStateAthena!(execute, Object=@n6) -> (then, CastFailed, AsFort Player State Athena:Fort Player State Athena Object Reference) {is=title K2Node_DynamicCast}
   > n9 = Cast To FortPlayerStateAthena(Object=Get PlayerState(self=GetOwningPlayerPawn!(self:User Widget Object Reference) -> (ReturnValue:Pawn Object Reference)))
   > Return Node(SameSquad="Equal (Byte)"(
     A:Byte=GetSquadId!(self:Fort Player State Athena Object Reference=@n9) -> (ReturnValue:Byte=0)
     B:Byte=GetSquadId(self=@n8) -> (ReturnValue=0)
    ) -> (ReturnValue:bool=~) {is=title K2Node_PromotableOperator})
  ]}
 UpdateEnergy {nodes [
   n1 = UpdateEnergy!() -> (then, Value:double=~, bIsResting:bool=~) {is=title K2Node_FunctionEntry}
   > n2 = GetDynamicMaterial(self=Get FillBar() -> (FillBar:Image Object Reference))
   > SetScalarParameterValue(self=@n2, ParameterName=Progress, Value=@n1)
   > n3 = GetDynamicMaterial(self=Get DeltaBar() -> (DeltaBar:Image Object Reference))
   > SetScalarParameterValue(self=@n3, ParameterName=Progress, Value=@n1)
   n4 = Get bIsResting() -> (bIsResting:bool)
   > SetColorAndOpacity(
    self:Image Object Reference=Get FillBar()
    InColorAndOpacity:Linear Color Structure=SelectColor(A=(R=1,G=0.0365,B=0.0365,A=1), B=(R=1,G=1,B=1,A=1), bPickA=@n4)
   )
   > n5 = Branch(Condition=@n4)
   > n6 = Branch(Condition=IsAnimationPlaying(InAnimation=Get RechargingAnim() -> (RechargingAnim:Widget Animation Object Reference)))
   PlayAnimation(execute=@n6.else, InAnimation=Get RechargingAnim(), NumLoopsToPlay=0)
   Branch(execute=@n5.else, Condition=IsAnimationPlaying(InAnimation=Get RechargingAnim()))
   > StopAnimation(InAnimation=Get RechargingAnim())
  ]}
]
variables [
 Anim_DeltaThreshold {type=number}
 FortPawn {type=FortPawn, default=null}
 RidableBaseComponent {type=LabradorRiding_CreatureBaseRidableComponent_C, default=null}
]
widgets {
 info {inheritedWidgetCount=0, parentClass=/Script/FortniteUI.NPCStatusWidgetExtensionBase, rootWidgetClass=&Overlay, widgetCount=5}
 widgets [
  Overlay_Stats {widgetClassPath=&Overlay, bIsVariable=true, widget=Overlay_Stats}
  SizeBox_Progress {
   widgetClassPath=&SizeBox
   bIsVariable=true
   parent=Overlay_Stats
   widget=SizeBox_Progress
   properties {bOverride_HeightOverride=true, heightOverride=6, widthOverride=200}
   slot {class=OverlaySlot, properties {horizontalAlignment=HAlign_Fill, verticalAlignment=VAlign_Fill}}
  }
  Overlay_2 {widgetClassPath=&Overlay, parent=SizeBox_Progress, widget=Overlay_2, slot {class=SizeBoxSlot}}
  DeltaBar {
   widgetClassPath=&Image
   bIsVariable=true
   parent=Overlay_2
   widget=DeltaBar
   properties {
    visibility=HitTestInvisible
    brush {
     bIsDynamicallyLoaded=false
     drawAs=Image
     imageType=NoImage
     mirroring=NoMirror
     resourceName=None
     resourceObject=/LabradorRidingContent/UI/Materials/M_LabradorRiding_UI_Animal_EnergyDelta
     tiling=NoTile
     imageSize=<32 32>
     margin=<ltrb 0 0 0 0>
     outlineSettings {
      bUseBrushTransparency=false
      roundingType=HalfHeightRadius
      width=0
      color {colorUseRule=UseColor_Specified, specifiedColor=<rgba 0 0 0 0>}
      cornerRadii=<0 0 0 0>
     }
     tintColor {colorUseRule=UseColor_Specified, specifiedColor=<rgba 1 1 1 1>}
     uVRegion {bIsValid=false, max=<0 0>, min=<0 0>}
    }
   }
   slot {class=OverlaySlot, properties {horizontalAlignment=HAlign_Fill}}
  }
  FillBar {
   widgetClassPath=&Image
   bIsVariable=true
   parent=Overlay_2
   widget=FillBar
   properties {brush {
     bIsDynamicallyLoaded=false
     drawAs=Image
     imageType=NoImage
     mirroring=NoMirror
     resourceName=None
     resourceObject=/LabradorRidingContent/UI/Materials/M_LabradorRiding_UI_Animal_Energy
     tiling=NoTile
     imageSize=<32 32>
     margin=<ltrb 0 0 0 0>
     outlineSettings {
      bUseBrushTransparency=false
      roundingType=HalfHeightRadius
      width=0
      color {colorUseRule=UseColor_Specified, specifiedColor=<rgba 0 0 0 0>}
      cornerRadii=<0 0 0 0>
     }
     tintColor {colorUseRule=UseColor_Specified, specifiedColor=<rgba 1 1 1 1>}
     uVRegion {bIsValid=false, max=<0 0>, min=<0 0>}
    }}
   slot {class=OverlaySlot, properties {horizontalAlignment=HAlign_Fill, verticalAlignment=VAlign_Center}}
  }
 ]
}
