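# HUD widget that shows the ridden animal's icon, health bar and health value in Labrador Riding.
# Rider component data supplies the icon and pawn, whose health-changed event updates the text and fill material.
# A looping timer steps a delayed delta bar down to current health.
# It shifts with overshield visibility and compacts for touch; its stamina bar and boost animations are present but undriven by these graphs.
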WidgetBlueprint /LabradorRidingContent/UI/WBP_LabradorRiding_HealthInfo
animations [
 Anim_FullBar {
  binding_count=1
  length_seconds=0.75
  object=&self:Anim_FullBar
  sequence {bindings [Overlay_Stats {
     id=#1
     possessed_class=Overlay
     tracks [{
       type=MovieSceneFloatTrack
       display_name=Render Opacity
       sections [{type=MovieSceneFloatSection, end_frame=15, start_frame=0, channels [None {type=MovieSceneScriptingFloatChannel, keys [[0,1],[10,1],[15,0]]}]}]
      }]
    }]}
 }
 Anim_Low {
  binding_count=1
  length_seconds=0.5
  object=&self:Anim_Low
  sequence {bindings [Overlay_Stats {
     id=#2
     possessed_class=Overlay
     tracks [{
       type=MovieSceneFloatTrack
       display_name=Render Opacity
       sections [{
         type=MovieSceneFloatSection
         end_frame=10
         start_frame=0
         channels [None {type=MovieSceneScriptingFloatChannel, keys [[0,1],[5,0.33],[10,1]]}]
        }]
      }]
    }]}
 }
 Anim_BoostStarted {
  binding_count=2
  length_seconds=0.4
  object=&self:Anim_BoostStarted
  sequence {
   bindings [
    Image_FillBar_Energy {
     id=#3
     possessed_class=Image
     tracks [{
       type=MovieSceneWidgetMaterialTrack
       sections [{
         type=MovieSceneParameterSection
         end_frame=24
         start_frame=0
         channels [
          Progress_Color_StopR {type=MovieSceneScriptingFloatChannel, const [1,2]}
          Progress_Color_StopG {type=MovieSceneScriptingFloatChannel, const [1,2]}
          Progress_Color_StopB {type=MovieSceneScriptingFloatChannel, keys [[0,0],[24,1]]}
          Progress_Color_StopA {type=MovieSceneScriptingFloatChannel, const [1,2]}
         ]
        }]
      }]
    }
    SizeBox_Stamina {
     id=#4
     possessed_class=SizeBox
     tracks [{
       type=MovieSceneFloatTrack
       display_name=Render Opacity
       sections [{type=MovieSceneFloatSection, end_frame=24, start_frame=0, channels [None {type=MovieSceneScriptingFloatChannel, keys [[0,0],[24,1]]}]}]
      }]
    }
   ]
   display_rate {numerator=60}
  }
 }
 Anim_BoostStopped {
  binding_count=3
  length_seconds=0.3
  object=&self:Anim_BoostStopped
  sequence {bindings [
    Overlay_BoostSparks {
     id=#5
     possessed_class=Overlay
     tracks [{
       type=MovieSceneFloatTrack
       display_name=Render Opacity
       sections [{type=MovieSceneFloatSection, end_frame=6, start_frame=0, channels [None {type=MovieSceneScriptingFloatChannel, keys [[0,1],[6,0]]}]}]
      }]
    }
    Border_BoostGlow {
     id=#6
     possessed_class=Border
     tracks [{
       type=MovieSceneFloatTrack
       display_name=Render Opacity
       sections [{type=MovieSceneFloatSection, end_frame=6, start_frame=0, channels [None {type=MovieSceneScriptingFloatChannel, keys [[0,1],[6,0]]}]}]
      }]
    }
    Image_FillBar_Energy {
     id=#7
     possessed_class=Image
     tracks [{
       type=MovieSceneWidgetMaterialTrack
       sections [{
         type=MovieSceneParameterSection
         end_frame=6
         start_frame=0
         channels [
          Progress_Color_StopR {type=MovieSceneScriptingFloatChannel, const [1,2]}
          Progress_Color_StopG {type=MovieSceneScriptingFloatChannel, const [1,2]}
          Progress_Color_StopB {type=MovieSceneScriptingFloatChannel, keys [[0,1],[6,0]]}
          Progress_Color_StopA {type=MovieSceneScriptingFloatChannel, const [1,2]}
         ]
        }]
      }]
    }
   ]}
 }
]
graphs [
 BindEventToOvershieldVisChanged {nodes [
   n1 = BindEventToOvershieldVisChanged!() -> (then, BindEvents:bool=~) {is=title K2Node_FunctionEntry}
   n2 = GetPlayerStateZone!(self:Fort HUDContext Object Reference=Get FortHUDContext!() -> (ReturnValue:Fort HUDContext Object Reference) {is=title K2Node_GetBlueprintContextEx}) -> (ReturnValue:Fort Player State Zone Object Reference)
   > Is Valid(InputObject=@n2)
   > n3 = Branch(Condition=@n1)
   n4 = Create Event()
   > Bind Event to On Overshield Visibility Changed(self:Fort Player State Zone Object Reference=@n2, Delegate=@n4) {is=title K2Node_AddDelegate}
   > OnOvershieldVisibilityChanged!(execute, self, IsVisible:bool=ShouldOvershieldBarBeVisible!(self:Fort Player State Zone Object Reference=@n2) -> (ReturnValue:bool)) -> (then)
   Unbind Event from On Overshield Visibility Changed(execute=@n3.else, self:Fort Player State Zone Object Reference=@n2, Delegate=@n4) {is=title K2Node_RemoveDelegate}
  ]}
 BindToRiderComponentEvents {nodes [
   BindToRiderComponentEvents!() -> (then, BindEvents:bool=~) {is=title K2Node_FunctionEntry}
   > n1 = Cast To Pawn!(execute, Object=GetViewTarget!(self:Controller Object Reference=GetOwningPlayer()) -> (ReturnValue:Actor Object Reference)) -> (then, CastFailed, AsPawn:Pawn Object Reference) {is=title K2Node_DynamicCast}
   > Is Valid(InputObject=@n1)
   > n2 = Set PlayerRiderComponentRef(PlayerRiderComponentRef:Labrador Riding Player Rider Component Object Reference=GetComponentByClass!(
     self:Actor Object Reference=@n1
     ComponentClass:Actor Component Class Reference=/LabradorRidingContent/Components/LabradorRiding_PlayerRiderComponent.LabradorRiding_PlayerRiderComponent_C
    ) -> (ReturnValue:Labrador Riding Player Rider Component Object Reference)) -> (Output_Get:Labrador Riding Player Rider Component Object Reference)
   > Is Valid(InputObject=@n2)
   > n3 = Branch(Condition=Get BindEvents() -> (BindEvents:bool))
   n4 = Create Event()
   n5 = Get PlayerRiderComponentRef() -> (PlayerRiderComponentRef:Labrador Riding Player Rider Component Object Reference)
   > Bind Event to Rider UINew Stats(self:Labrador Riding Player Rider Component Object Reference=@n5, Delegate=@n4) {is=title K2Node_AddDelegate}
   Unbind Event from Rider UINew Stats(execute=@n3.else, self:Labrador Riding Player Rider Component Object Reference=@n5, Delegate=@n4) {is=title K2Node_RemoveDelegate}
  ]}
 ClearTimerHandles {nodes [
   ClearTimerHandles() {is=title K2Node_FunctionEntry}
   > Clear and Invalidate Timer by Handle(Handle=Get HealthDelta_TimerHandle() -> (HealthDelta_TimerHandle:Timer Handle Structure))
  ]}
 EventGraph {nodes [
   n1 = OnRidablePawnHealthChanged() {is=title K2Node_CustomEvent}
   > UpdateHealthChanged!(execute, self, SkipDeltaAnim:bool=false) -> (then)
   n2 = UpdateRidableActorData!() -> (OutputDelegate, then, Name:Text, Icon:Slate Brush Structure, RidableActor:Actor Object Reference) {is=title K2Node_CustomEvent}
   > n3 = Cast To FortPawn!(execute, Object=@n2.RidableActor) -> (then, CastFailed, AsFort Pawn:Fort Pawn Object Reference) {is=title K2Node_DynamicCast}
   > n4 = Set RidablePawn(RidablePawn:Fort Pawn Object Reference=@n3) -> (Output_Get:Fort Pawn Object Reference)
   > SetIcon!(execute, self, InIconBrush:"Slate Brush Structure (by ref)"=@n2.Icon) -> (then)
   > Bind Event to On Pawn Health Changed(self:Fort Pawn Object Reference=@n4, Delegate=@n1) {is=title K2Node_AddDelegate}
   > UpdateHealthChanged(SkipDeltaAnim=true)
   Event Destruct()
   > BindToRiderComponentEvents!(execute, self, BindEvents:bool=false) -> (then) {is=title K2Node_CallFunction}
   > BindEventToOvershieldVisChanged!(execute, self, BindEvents:bool=false) -> (then) {is=title K2Node_CallFunction}
   > ClearTimerHandles() {is=title K2Node_CallFunction}
   Event Construct()
   > BindToRiderComponentEvents(BindEvents=true)
   > BindEventToOvershieldVisChanged(BindEvents=true)
   > Initialize!(
    execute
    self:Common Tag Visibility Widget Object Reference=Get TagVisibility_HealthValue() -> (TagVisibility_HealthValue:Common Tag Visibility Widget Object Reference)
    OwningLocalPlayer:Local Player Object Reference=GetOwningLocalPlayer!(self:Widget Object Reference) -> (ReturnValue:Local Player Object Reference)
    OwningPlayerController:Player Controller Object Reference=GetOwningPlayer()
   ) -> (then)
   > n5 = GetRidableActorData!(execute, self:Labrador Riding Player Rider Component Object Reference=Get PlayerRiderComponentRef()) -> (then, Success:bool, RidableActorName:Text, RidableActorIcon:Slate Brush Structure, CachedRidableActor:Actor Object Reference)
   > Branch(Condition=@n5)
   > UpdateRidableActorData!(
    execute
    self
    Name:Text=@n5.RidableActorName
    Icon:Slate Brush Structure=@n5.RidableActorIcon
    RidableActor:Actor Object Reference=@n5.CachedRidableActor
   ) -> (then) {is=title K2Node_CallFunction}
   Event OnInitialized()
   > SetCompactUIStyle() {is=title K2Node_CallFunction}
  ]}
 HealthChaseBarInterpolate {nodes [
   HealthChaseBarInterpolate() {is=title K2Node_FunctionEntry}
   > n1 = Set CachedInterpolateHealth(CachedInterpolateHealth:double=float - float(A=Get CachedInterpolateHealth() -> (CachedInterpolateHealth:double), B=0.01)) -> (Output_Get:double)
   > n2 = Branch(Condition="float <= float"(A=@n1, B=Get CurrentHealth() -> (CurrentHealth:double)) -> (ReturnValue=~))
   > Clear and Invalidate Timer by Handle(Handle=Get HealthDelta_TimerHandle())
   > n3 = GetDynamicMaterial(self=Get Image_DeltaBar_Health() -> (Image_DeltaBar_Health:Image Object Reference))
   > SetScalarParameterValue(self=@n3, ParameterName=Progress, Value=Get CurrentHealth())
   > Set CachedInterpolateHealth(CachedInterpolateHealth=Get CurrentHealth())
   n4 = GetDynamicMaterial(execute=@n2.else, self=Get Image_DeltaBar_Health())
   > SetScalarParameterValue(self=@n4, ParameterName=Progress, Value=Get CachedInterpolateHealth())
  ]}
 OnOvershieldVisibilityChanged {nodes [
   n1 = OnOvershieldVisibilityChanged!() -> (then, IsVisible:bool=~) {is=title K2Node_FunctionEntry}
   n2 = GetOwningPlayer()
   > Branch(Condition=NOT Boolean(A=IsUsingTouch(PlayerController=@n2)))
   > SetRenderTranslation(Translation=MakeVector2D(X=0, Y=SelectFloat(A=0, B=15, bPickA=@n1)))
  ]}
 SetCompactUIStyle {nodes [
   SetCompactUIStyle() {is=title K2Node_FunctionEntry}
   > Branch(Condition=IsUsingInputType!(InputType:ECommonInputType Enum=Touch, PlayerController:Player Controller Object Reference=GetOwningPlayer()) -> (ReturnValue:bool))
   n1 = Get SizeBox_Stamina() -> (SizeBox_Stamina:Size Box Object Reference)
   > SetHeightOverride(self=@n1, InHeightOverride=15)
   > SetWidthOverride(self=@n1, InWidthOverride=275)
   > SetWidthOverride(self=Get SizeBox_Health() -> (SizeBox_Health:Size Box Object Reference), InWidthOverride=258)
   > SetPadding!(
    execute
    self:Horizontal Box Slot Object Reference=SlotAsHorizontalBoxSlot!(Widget:Widget Object Reference=Get Image_Icon() -> (Image_Icon:Image Object Reference)) -> (ReturnValue:Horizontal Box Slot Object Reference)
    InPadding_Left:float=0
    InPadding_Top:float=0
    InPadding_Right:float=0
    InPadding_Bottom:float=0
   ) -> (then)
   > SetRenderTransform!(
    execute
    self:Widget Object Reference=Get Image_Icon()
    InTransform_Translation:Vector 2D Structure=(X=16,Y=0)
    InTransform_Scale:Vector 2D Structure=(X=0.37,Y=0.37)
    InTransform_Shear:Vector 2D Structure=(X=0,Y=0)
    InTransform_Angle:float=0
   ) -> (then)
  ]}
 SetIcon {nodes [
   n1 = SetIcon!() -> (then, InIconBrush:"Slate Brush Structure (by ref)") {is=title K2Node_FunctionEntry}
   > SetBrush(
    self=Get Image_Icon()
    InBrush=Make Slate Brush!(
     TintColor:Slate Color Structure=(SpecifiedColor=(R=1,G=1,B=1,A=1),ColorUseRule=UseColor_Specified)
     DrawAs:ESlateBrushDrawType Enum=Image
     Tiling:ESlateBrushTileType Enum=NoTile
     Mirroring:ESlateBrushMirrorType Enum=NoMirror
     ImageSize_X:double=24
     ImageSize_Y:double=24
     Margin:Margin Structure=(Left=0,Top=0,Right=0,Bottom=0)
     ResourceObject:Object Reference=Break Slate Brush!(SlateBrush:"Slate Brush Structure (by ref)"=@n1) -> (
      TintColor:Slate Color Structure
      DrawAs:ESlateBrushDrawType Enum
      Tiling:ESlateBrushTileType Enum
      Mirroring:ESlateBrushMirrorType Enum
      ImageSize:Vector 2D Structure
      Margin:Margin Structure
      ResourceObject:Object Reference
      OutlineSettings:Slate Brush Outline Settings Structure
     ) {is=title K2Node_BreakStruct}.ResourceObject
     OutlineSettings:Slate Brush Outline Settings Structure=(CornerRadii=(X=0,Y=0,Z=0,W=0),Color=(SpecifiedColor=(R=0,G=0,B=0,A=0),ColorUseRule=UseColor_Specified),Width=0,RoundingType=HalfHeightRadius,bUseBrushTransparency=False)
    ) -> (SlateBrush:Slate Brush Structure) {is=title K2Node_MakeStruct}
   )
  ]}
 UpdateHealthChanged {nodes [
   UpdateHealthChanged!() -> (then, SkipDeltaAnim:bool=~) {is=title K2Node_FunctionEntry}
   n1 = Get RidablePawn() -> (RidablePawn:Fort Pawn Object Reference)
   n2 = GetHealth!(self:Fort Pawn Object Reference=@n1) -> (ReturnValue:float)
   > "SetText (Text)"(
    self=Get Text_CurrentValue() -> (Text_CurrentValue:Common Text Object Reference)
    InText="To Text (Float)"(Value=@n2, MinimumIntegralDigits=1, MaximumIntegralDigits=324, MinimumFractionalDigits=0, MaximumFractionalDigits=0)
   )
   > n3 = Set CurrentHealth(CurrentHealth:double=float / float(A=@n2, B=GetMaxHealth!(self:Fort Pawn Object Reference=@n1) -> (ReturnValue:float))) -> (Output_Get:double)
   > n4 = GetDynamicMaterial(self=Get Image_FillBar_Health() -> (Image_FillBar_Health:Image Object Reference))
   > SetScalarParameterValue(self=@n4, ParameterName=Progress, Value=@n3)
   > n5 = Branch(Condition=Get SkipDeltaAnim() -> (SkipDeltaAnim:bool))
   > n6 = Branch(Condition=Is Timer Active by Handle!(Handle:Timer Handle Structure=Get HealthDelta_TimerHandle()) -> (ReturnValue:bool))
   > n7 = Clear and Invalidate Timer by Handle(Handle=Get HealthDelta_TimerHandle())
   n8 = Set CachedInterpolateHealth(execute=[@n7.then, @n6.else], CachedInterpolateHealth=Get CurrentHealth())
   > n9 = GetDynamicMaterial(self=Get Image_DeltaBar_Health())
   > SetScalarParameterValue(self=@n9, ParameterName=Progress, Value=@n8)
   n10 = Set Timer by Event(execute=@n5.else, Delegate=Create Event(), Time=0.02, bLooping=true, InitialStartDelay=0.1)
   > Set HealthDelta_TimerHandle(HealthDelta_TimerHandle:Timer Handle Structure=@n10) -> (Output_Get:Timer Handle Structure)
  ]}
]
variables [
 BoostingActive {type=boolean}
 CachedInterpolateHealth {type=number, default=1}
 CurrentHealth {type=number}
 HealthDelta_TimerHandle {type=TimerHandle}
 PlayerRiderComponentRef {type=LabradorRiding_PlayerRiderComponent_C, default=null}
 RidablePawn {type=FortPawn, default=null}
 SprintingAbility_GameplayTag {type=GameplayTag, default {tagName=Gameplay.Action.Npc.RidingSprinting}}
 SprintTagAdded_AsyncAction {type=AbilityAsync_WaitGameplayTagAdded, default=null}
 SprintTagRemoved_AsyncAction {type=AbilityAsync_WaitGameplayTagRemoved, default=null}
]
widgets {
 info {inheritedWidgetCount=1, parentClass=&FortHUDElementWidget, rootWidgetClass=&Overlay, widgetCount=24}
 widgets [
  Overlay_Stats {widgetClassPath=&Overlay, widget=Overlay_Stats}
  HorizontalBox_50 {
   widgetClassPath=&HorizontalBox
   parent=Overlay_Stats
   widget=HorizontalBox_50
   slot {class=OverlaySlot, properties {verticalAlignment=VAlign_Fill}}
  }
  Image_Icon {
   widgetClassPath=&Image
   bIsVariable=true
   parent=HorizontalBox_50
   widget=Image_Icon
   properties {brush {
     bIsDynamicallyLoaded=false
     drawAs=Image
     imageType=FullColor
     mirroring=NoMirror
     resourceName=None
     resourceObject=/Game/Athena/HUD/Art/T_UI_Llama_Icon_64x
     tiling=NoTile
     imageSize=<24 24>
     margin=<ltrb 0.2 0.2 0.2 0.2>
     outlineSettings {
      bUseBrushTransparency=false
      roundingType=HalfHeightRadius
      width=0
      color {colorUseRule=UseColor_Specified, specifiedColor=<rgba 0 0 0 0>}
      cornerRadii=<0 0 0 0>
     }
     tintColor {colorUseRule=UseColor_Specified, specifiedColor=<rgba 1 1 1 1>}
     uVRegion {bIsValid=false, max=<0 0>, min=<0 0>}
    }}
   slot {class=HorizontalBoxSlot, properties {horizontalAlignment=HAlign_Right, verticalAlignment=VAlign_Bottom, padding=<ltrb 0 0 3 0>}}
  }
  VerticalBox_85 {
   widgetClassPath=&VerticalBox
   parent=HorizontalBox_50
   widget=VerticalBox_85
   slot {class=HorizontalBoxSlot, properties {horizontalAlignment=HAlign_Right, verticalAlignment=VAlign_Center}}
  }
  SizeBox_Stamina {
   widgetClassPath=&SizeBox
   bIsVariable=true
   parent=VerticalBox_85
   widget=SizeBox_Stamina
   properties {bOverride_HeightOverride=true, bOverride_WidthOverride=true, heightOverride=6, renderOpacity=0, visibility=Collapsed, widthOverride=333}
   slot {class=VerticalBoxSlot, properties {horizontalAlignment=HAlign_Left, verticalAlignment=VAlign_Top, padding=<ltrb 0 0 0 6>}}
  }
  Overlay_2 {widgetClassPath=&Overlay, parent=SizeBox_Stamina, widget=Overlay_2, slot {class=SizeBoxSlot}}
  Image_EmptyStatusGlow {
   widgetClassPath=&Image
   bIsVariable=true
   parent=Overlay_2
   widget=Image_EmptyStatusGlow
   properties {
    renderOpacity=0
    brush {
     bIsDynamicallyLoaded=false
     drawAs=Box
     imageType=NoImage
     mirroring=NoMirror
     resourceName=None
     resourceObject=/Game/Athena/HUD/LTM/Wax/Materials/M_UI_BallerBoostGlow
     tiling=NoTile
     imageSize=<32 32>
     margin=<ltrb 0.5 0.5 0.5 0.5>
     outlineSettings {
      bUseBrushTransparency=false
      roundingType=HalfHeightRadius
      width=0
      color {colorUseRule=UseColor_Specified, specifiedColor=<rgba 0 0 0 0>}
      cornerRadii=<0 0 0 0>
     }
     tintColor {colorUseRule=UseColor_Specified, specifiedColor=<rgba 1 1 1 1>}
     uVRegion {bIsValid=false, max=<0 0>, min=<0 0>}
    }
    renderTransform {angle=0, scale=<-1 1>, shear=<0 0>, translation=<0 0>}
   }
   slot {class=OverlaySlot, properties {horizontalAlignment=HAlign_Fill, verticalAlignment=VAlign_Fill, padding=<ltrb -13 -12 128 -12>}}
  }
  Image_DeltaBar_Energy {
   widgetClassPath=&Image
   bIsVariable=true
   parent=Overlay_2
   widget=Image_DeltaBar_Energy
   properties {
    visibility=HitTestInvisible
    brush {
     bIsDynamicallyLoaded=false
     drawAs=Image
     imageType=NoImage
     mirroring=NoMirror
     resourceName=None
     resourceObject=/LabradorRidingContent/UI/Materials/M_LabradorRiding_UI_Animal_EnergyDelta
     tiling=NoTile
     imageSize=<32 32>
     margin=<ltrb 0 0 0 0>
     outlineSettings {
      bUseBrushTransparency=false
      roundingType=HalfHeightRadius
      width=0
      color {colorUseRule=UseColor_Specified, specifiedColor=<rgba 0 0 0 0>}
      cornerRadii=<0 0 0 0>
     }
     tintColor {colorUseRule=UseColor_Specified, specifiedColor=<rgba 1 1 1 1>}
     uVRegion {bIsValid=false, max=<0 0>, min=<0 0>}
    }
   }
   slot {class=OverlaySlot, properties {horizontalAlignment=HAlign_Fill}}
  }
  Image_FillBar_Energy {
   widgetClassPath=&Image
   bIsVariable=true
   parent=Overlay_2
   widget=Image_FillBar_Energy
   properties {brush {
     bIsDynamicallyLoaded=false
     drawAs=Image
     imageType=NoImage
     mirroring=NoMirror
     resourceName=None
     resourceObject=/LabradorRidingContent/UI/Materials/M_LabradorRiding_UI_Animal_Energy
     tiling=NoTile
     imageSize=<32 32>
     margin=<ltrb 0 0 0 0>
     outlineSettings {
      bUseBrushTransparency=false
      roundingType=HalfHeightRadius
      width=0
      color {colorUseRule=UseColor_Specified, specifiedColor=<rgba 0 0 0 0>}
      cornerRadii=<0 0 0 0>
     }
     tintColor {colorUseRule=UseColor_Specified, specifiedColor=<rgba 1 1 1 1>}
     uVRegion {bIsValid=false, max=<0 0>, min=<0 0>}
    }}
   slot {class=OverlaySlot, properties {horizontalAlignment=HAlign_Fill, verticalAlignment=VAlign_Center}}
  }
  Border_BoostGlow {
   widgetClassPath=&Border
   bIsVariable=true
   parent=Overlay_2
   widget=Border_BoostGlow
   properties {
    renderOpacity=0
    background {
     bIsDynamicallyLoaded=false
     drawAs=NoDrawType
     imageType=NoImage
     mirroring=NoMirror
     resourceName=None
     resourceObject=null
     tiling=NoTile
     imageSize=<32 32>
     margin=<ltrb 0 0 0 0>
     outlineSettings {
      bUseBrushTransparency=false
      roundingType=HalfHeightRadius
      width=0
      color {colorUseRule=UseColor_Specified, specifiedColor=<rgba 0 0 0 0>}
      cornerRadii=<0 0 0 0>
     }
     tintColor {colorUseRule=UseColor_Specified, specifiedColor=<rgba 1 1 1 1>}
     uVRegion {bIsValid=false, max=<0 0>, min=<0 0>}
    }
    desiredSizeScale=<0.5 1>
    padding=<ltrb 0 0 0 0>
    renderTransformPivot=<0 0.5>
   }
   slot {class=OverlaySlot, properties {horizontalAlignment=HAlign_Fill, verticalAlignment=VAlign_Fill}}
  }
  SizeBox_1 {
   widgetClassPath=&SizeBox
   parent=Border_BoostGlow
   widget=SizeBox_1
   properties {bOverride_WidthOverride=true, widthOverride=300}
   slot {class=BorderSlot, properties {padding=<ltrb 0 0 0 0>}}
  }
  Image_BoostGlow {
   widgetClassPath=&Image
   parent=SizeBox_1
   widget=Image_BoostGlow
   properties {
    brush {
     bIsDynamicallyLoaded=false
     drawAs=Box
     imageType=NoImage
     mirroring=NoMirror
     resourceName=None
     resourceObject=&M_UI_BallerBoostGlow
     tiling=NoTile
     imageSize=<72 72>
     margin=<ltrb 0.5 0.5 0.5 0.5>
     outlineSettings {
      bUseBrushTransparency=false
      roundingType=HalfHeightRadius
      width=0
      color {colorUseRule=UseColor_Specified, specifiedColor=<rgba 0 0 0 0>}
      cornerRadii=<0 0 0 0>
     }
     tintColor {colorUseRule=UseColor_Specified, specifiedColor=<rgba 1 1 1 1>}
     uVRegion {bIsValid=false, max=<0 0>, min=<0 0>}
    }
    renderTransformPivot=<0 0.5>
   }
   slot {class=SizeBoxSlot, properties {padding=<ltrb -32 -32 -32 -32>}}
  }
  Overlay_BoostSparks {
   widgetClassPath=&Overlay
   bIsVariable=true
   parent=Overlay_2
   widget=Overlay_BoostSparks
   properties {renderOpacity=0, renderTransform {angle=0, scale=<1 1>, shear=<0 0>, translation=<-1 0>}, renderTransformPivot=<0 0.5>}
   slot {class=OverlaySlot, properties {horizontalAlignment=HAlign_Right, verticalAlignment=VAlign_Center}}
  }
  Image_LargeCone {
   widgetClassPath=&Image
   parent=Overlay_BoostSparks
   widget=Image_LargeCone
   properties {
    visibility=HitTestInvisible
    brush {
     bIsDynamicallyLoaded=false
     drawAs=Image
     imageType=NoImage
     mirroring=NoMirror
     resourceName=None
     resourceObject=/Game/Athena/HUD/Vehicles/Materials/M_UI_SparksGlow_Inst
     tiling=NoTile
     imageSize=<150 200>
     margin=<ltrb 0 0 0 0>
     outlineSettings {
      bUseBrushTransparency=false
      roundingType=HalfHeightRadius
      width=0
      color {colorUseRule=UseColor_Specified, specifiedColor=<rgba 0 0 0 0>}
      cornerRadii=<0 0 0 0>
     }
     tintColor {colorUseRule=UseColor_Specified, specifiedColor=<rgba 1 1 1 1>}
     uVRegion {bIsValid=false, max=<0 0>, min=<0 0>}
    }
    renderTransform {angle=-12, scale=<1 1>, shear=<0 0>, translation=<0 28>}
   }
   slot {class=OverlaySlot, properties {horizontalAlignment=HAlign_Center, verticalAlignment=VAlign_Center, padding=<ltrb -200 -200 -200 -200>}}
  }
  Image_Sparks_1 {
   widgetClassPath=&Image
   parent=Overlay_BoostSparks
   widget=Image_Sparks_1
   properties {
    visibility=HitTestInvisible
    brush {
     bIsDynamicallyLoaded=false
     drawAs=Image
     imageType=NoImage
     mirroring=NoMirror
     resourceName=None
     resourceObject=/Game/Athena/UI/InGame/Interrogation/Materials/M_UI_Starburst_Sparks_Inst
     tiling=NoTile
     imageSize=<120 400>
     margin=<ltrb 0 0 0 0>
     outlineSettings {
      bUseBrushTransparency=false
      roundingType=HalfHeightRadius
      width=0
      color {colorUseRule=UseColor_Specified, specifiedColor=<rgba 0 0 0 0>}
      cornerRadii=<0 0 0 0>
     }
     tintColor {colorUseRule=UseColor_Specified, specifiedColor=<rgba 1 0.7183 0.651 1>}
     uVRegion {bIsValid=false, max=<0 0>, min=<0 0>}
    }
    renderTransform {angle=-12, scale=<1 1>, shear=<0 0>, translation=<0 0>}
   }
   slot {class=OverlaySlot, properties {horizontalAlignment=HAlign_Center, verticalAlignment=VAlign_Center, padding=<ltrb -200 -200 -200 -200>}}
  }
  Image_Sparks {
   widgetClassPath=&Image
   parent=Overlay_BoostSparks
   widget=Image_Sparks
   properties {
    visibility=HitTestInvisible
    brush {
     bIsDynamicallyLoaded=false
     drawAs=Image
     imageType=NoImage
     mirroring=NoMirror
     resourceName=None
     resourceObject=/Game/Athena/UI/InGame/Interrogation/Materials/M_UI_Starburst_Sparks
     tiling=NoTile
     imageSize=<295.3884 400>
     margin=<ltrb 0 0 0 0>
     outlineSettings {
      bUseBrushTransparency=false
      roundingType=HalfHeightRadius
      width=0
      color {colorUseRule=UseColor_Specified, specifiedColor=<rgba 0 0 0 0>}
      cornerRadii=<0 0 0 0>
     }
     tintColor {colorUseRule=UseColor_Specified, specifiedColor=<rgba 1 0.3682 0.0743 1>}
     uVRegion {bIsValid=false, max=<0 0>, min=<0 0>}
    }
    renderTransform {angle=-12, scale=<1 1>, shear=<0 0>, translation=<0 0>}
   }
   slot {class=OverlaySlot, properties {horizontalAlignment=HAlign_Center, verticalAlignment=VAlign_Center, padding=<ltrb -200 -200 -200 -200>}}
  }
  HorizontalBox_0 {
   widgetClassPath=&HorizontalBox
   parent=VerticalBox_85
   widget=HorizontalBox_0
   slot {class=VerticalBoxSlot, properties {verticalAlignment=VAlign_Bottom, size {sizeRule=Fill, value=1}}}
  }
  SizeBox_Health {
   widgetClassPath=&SizeBox
   bIsVariable=true
   parent=HorizontalBox_0
   widget=SizeBox_Health
   properties {bOverride_HeightOverride=true, bOverride_WidthOverride=true, heightOverride=24, widthOverride=333}
   slot {class=HorizontalBoxSlot}
  }
  Overlay {widgetClassPath=&Overlay, parent=SizeBox_Health, widget=Overlay, slot {class=SizeBoxSlot}}
  Image_DeltaBar_Health {
   widgetClassPath=&Image
   bIsVariable=true
   parent=Overlay
   widget=Image_DeltaBar_Health
   properties {
    visibility=HitTestInvisible
    brush {
     bIsDynamicallyLoaded=false
     drawAs=Image
     imageType=NoImage
     mirroring=NoMirror
     resourceName=None
     resourceObject=/LabradorRidingContent/UI/Materials/M_LabradorRiding_UI_Animal_HealthDelta
     tiling=NoTile
     imageSize=<32 32>
     margin=<ltrb 0 0 0 0>
     outlineSettings {
      bUseBrushTransparency=false
      roundingType=HalfHeightRadius
      width=0
      color {colorUseRule=UseColor_Specified, specifiedColor=<rgba 0 0 0 0>}
      cornerRadii=<0 0 0 0>
     }
     tintColor {colorUseRule=UseColor_Specified, specifiedColor=<rgba 1 1 1 1>}
     uVRegion {bIsValid=false, max=<0 0>, min=<0 0>}
    }
   }
   slot {class=OverlaySlot, properties {horizontalAlignment=HAlign_Fill}}
  }
  Image_FillBar_Health {
   widgetClassPath=&Image
   bIsVariable=true
   parent=Overlay
   widget=Image_FillBar_Health
   properties {brush {
     bIsDynamicallyLoaded=false
     drawAs=Image
     imageType=NoImage
     mirroring=NoMirror
     resourceName=None
     resourceObject=/LabradorRidingContent/UI/Materials/M_LabradorRiding_UI_Animal_Health
     tiling=NoTile
     imageSize=<32 32>
     margin=<ltrb 0 0 0 0>
     outlineSettings {
      bUseBrushTransparency=false
      roundingType=HalfHeightRadius
      width=0
      color {colorUseRule=UseColor_Specified, specifiedColor=<rgba 0 0 0 0>}
      cornerRadii=<0 0 0 0>
     }
     tintColor {colorUseRule=UseColor_Specified, specifiedColor=<rgba 1 1 1 1>}
     uVRegion {bIsValid=false, max=<0 0>, min=<0 0>}
    }}
   slot {class=OverlaySlot, properties {horizontalAlignment=HAlign_Fill, verticalAlignment=VAlign_Center}}
  }
  TagVisibility_HealthValue {
   widgetClassPath=/Script/FortniteUI.CommonTagVisibilityWidget
   bIsVariable=true
   parent=HorizontalBox_0
   widget=TagVisibility_HealthValue
   properties {hUDElementTag {gameplayTags [{tagName=HUD.BottomLeft.HitPointValues}]}}
   slot {class=HorizontalBoxSlot, properties {horizontalAlignment=HAlign_Left, verticalAlignment=VAlign_Center, padding=<ltrb 7 0 0 0>}}
  }
  Text_CurrentValue {
   widgetClassPath=&CommonTextBlock
   bIsVariable=true
   parent=TagVisibility_HealthValue
   widget=Text_CurrentValue
   properties {
    justification=Right
    lineHeightPercentage=0.8
    text=999?
    visibility=HitTestInvisible
    font {
     bForceMonospaced=false
     bMaterialIsStencil=false
     fontMaterial=null
     fontObject=&BurbankSmall
     letterSpacing=0
     monospacedWidth=1
     size=15
     skewAmount=0
     typefaceFontName=Black
     outlineSettings {
      bApplyOutlineToDropShadows=true
      bMiteredCorners=false
      bSeparateFillAlpha=true
      outlineMaterial=&M-FontShadow
      outlineSize=2
      outlineColor=<rgba 0 0 0 0.5>
     }
    }
    margin=<ltrb 0 -6 0 1>
    renderTransformPivot=<0 0.75>
    shadowOffset=<0 0>
   }
   slot {class=PanelSlot}
  }
  TutorialHighlight {widgetClassPath=&TutorialHighlight}
 ]
}
