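# Prioritized HUD widget that shows a "Llama Taming" progress bar for the ridable Labrador AI pawn.
# It listens for a stateful widget-progress event, binds to the pawn's Training Increased delegate, and toggles active content from the Visible flag.
# Progress updates set the bar material's Progress parameter and slide a Detail image, which is hidden by default.
# Showing plays InAnimation; ending plays Hurt and starts a one-second timer, apparently for hiding.
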WidgetBlueprint /LabradorRidingContent/UI/WBP_LabradorRiding_TrainingProgress
animations [
 Hurt {
  binding_count=1
  length_seconds=0.3
  object=&self:Hurt
  sequence {bindings [Overlay_0 {
     id=#1
     possessed_class=Overlay
     tracks [{
       type=MovieScene2DTransformTrack
       sections [{
         type=MovieScene2DTransformSection
         end_frame=6
         start_frame=0
         channels [
          Translation.X {type=MovieSceneScriptingFloatChannel, const [0,1]}
          Translation.Y {type=MovieSceneScriptingFloatChannel, const [0,1]}
          Angle {type=MovieSceneScriptingFloatChannel, const [0,1]}
          Scale.X {type=MovieSceneScriptingFloatChannel, keys [[0,1],[2,0.95],[4,1.02],[6,1]]}
          Scale.Y {type=MovieSceneScriptingFloatChannel, keys [[0,1],[2,0.92],[4,1.04],[6,1]]}
          Shear.X {type=MovieSceneScriptingFloatChannel, const [0,1]}
          Shear.Y {type=MovieSceneScriptingFloatChannel, const [0,1]}
         ]
        }]
      }]
    }]}
 }
 InAnimation {
  binding_count=2
  length_seconds=0.5
  object=&self:InAnimation
  sequence {bindings [
    TextContainer {
     id=#2
     possessed_class=Overlay
     tracks [
      {
       type=MovieScene2DTransformTrack
       sections [{
         type=MovieScene2DTransformSection
         end_frame=10
         start_frame=0
         channels [
          Translation.X {type=MovieSceneScriptingFloatChannel, const [0,1]}
          Translation.Y {type=MovieSceneScriptingFloatChannel, keys [[0,16.7514],[6,5],[10,0]]}
          Angle {type=MovieSceneScriptingFloatChannel, const [0,1]}
          Scale.X {type=MovieSceneScriptingFloatChannel, const [1,1]}
          Scale.Y {type=MovieSceneScriptingFloatChannel, keys [[0,0.5],[6,0.2],[8,1.05],[10,1]]}
          Shear.X {type=MovieSceneScriptingFloatChannel, const [0,1]}
          Shear.Y {type=MovieSceneScriptingFloatChannel, const [0,1]}
         ]
        }]
      }
      {
       type=MovieSceneFloatTrack
       display_name=Render Opacity
       sections [{
         type=MovieSceneFloatSection
         end_frame=6
         start_frame=0
         channels [None {type=MovieSceneScriptingFloatChannel, keys [[0,0],[5.9653,0],[6,1]]}]
        }]
      }
     ]
    }
    Overlay_0 {
     id=#3
     possessed_class=Overlay
     tracks [
      {
       type=MovieScene2DTransformTrack
       sections [{
         type=MovieScene2DTransformSection
         end_frame=10
         start_frame=0
         channels [
          Translation.X {type=MovieSceneScriptingFloatChannel, const [0,1]}
          Translation.Y {type=MovieSceneScriptingFloatChannel, keys [[0,-100],[9,2.5292],[10,0]]}
          Angle {type=MovieSceneScriptingFloatChannel, const [0,1]}
          Scale.X {type=MovieSceneScriptingFloatChannel, const [1,1]}
          Scale.Y {type=MovieSceneScriptingFloatChannel, keys [[0,1],[4,1.8],[6,1.7],[7,1.3],[9,0.9],[10,1]]}
          Shear.X {type=MovieSceneScriptingFloatChannel, const [0,1]}
          Shear.Y {type=MovieSceneScriptingFloatChannel, const [0,1]}
         ]
        }]
      }
      {
       type=MovieSceneFloatTrack
       display_name=Render Opacity
       sections [{type=MovieSceneFloatSection, end_frame=10, start_frame=0, channels [None {type=MovieSceneScriptingFloatChannel, keys [[0,0],[6,1]]}]}]
      }
     ]
    }
   ]}
 }
]
graphs [
 EventGraph {nodes [
   PlayAnimationReverse(InAnimation=Get InAnimation() -> (InAnimation:Widget Animation Object Reference), PlaybackSpeed=3)
   > Unbind Event from OnProgressUpdated(self:Object Reference=Reroute Node(InputPin:Object Reference) -> (OutputPin:Object Reference), Delegate=Create Event()) {is=title K2Node_RemoveDelegate}
   > SetHasActiveContent!(execute, self:Fort Prioritized Widget Interface, bNewActive:bool=false) -> (then)
   n1 = OnProgressUpdated_Event!() -> (OutputDelegate, then, UpdatedProgress:double=~) {is=title K2Node_CustomEvent}
   > Set Target(Target:double=@n1.UpdatedProgress) -> (Output_Get:double)
   > n2 = GetDynamicMaterial(self=Get ProgressBar() -> (ProgressBar:Image Object Reference))
   > SetScalarParameterValue(self=@n2, ParameterName=Progress, Value=Get Target() -> (Target:double))
   n3 = Lerp(A=-12, B=338, Alpha=Get Target())
   > SetRenderTransform!(
    execute
    self:Widget Object Reference=Get Detail() -> (Detail:Image Object Reference)
    InTransform:Widget Transform Structure=MakeTransform!(Translation:Vector 2D Structure=MakeVector2D(X=@n3, Y=-4), Scale:Vector 2D Structure=(X=1,Y=1), Shear:Vector 2D Structure, Angle:double=0) -> (ReturnValue:Widget Transform Structure)
   ) -> (then)
   EndProgress() {is=title K2Node_CustomEvent}
   > PlayAnimationForward(InAnimation=Get Hurt() -> (Hurt:Widget Animation Object Reference))
   > n4 = Set Timer by Event(Delegate=Create Event(), Time=1, bLooping=false)
   > Set Handle_HideProgress(Handle_HideProgress:Timer Handle Structure=@n4) -> (Output_Get:Timer Handle Structure)
   ShowProgress() {is=title K2Node_CustomEvent}
   > SetRenderOpacity(self=Get Overlay_0() -> (Overlay_0:Overlay Object Reference), InOpacity=0)
   > n5 = Branch(Condition=Is Timer Active by Handle!(Handle:Timer Handle Structure=Get Handle_HideProgress() -> (Handle_HideProgress:Timer Handle Structure)) -> (ReturnValue:bool))
   > n6 = Clear and Invalidate Timer by Handle(Handle=Get Handle_HideProgress())
   PlayAnimationForward(execute=[@n6.then, @n5.else], InAnimation=Get InAnimation())
   HideProgress() {is=title K2Node_CustomEvent}
   n7 = Get Overlay_0()
   > Branch(Condition="Equal (Float)"(A=GetRenderOpacity!(self:Widget Object Reference=@n7) -> (ReturnValue:float), B=1))
   > PlayAnimationReverse(InAnimation=Get InAnimation())
   Event Destruct()
   > UnbindEvents()
   Event Construct()
   > SetRenderOpacity(self=Get Overlay_0(), InOpacity=0)
   n8 = GetOwningPlayer()
   > Is Valid(InputObject=@n8)
   > n9 = Cast To FortPlayerController(Object=@n8)
   > n10 = Set CachedPC(CachedPC:Fort Player Controller Object Reference=@n9) -> (Output_Get:Fort Player Controller Object Reference)
   > Bind Event to On View Target Changed(self:Fort Player Controller Object Reference=@n10, Delegate=Create Event()) {is=title K2Node_AddDelegate}
   > OnViewTargetChanged!(
    execute
    self
    PlayerController:Fort Player Controller Object Reference
    OldViewTarget:Actor Object Reference
    NewViewTarget:Actor Object Reference=GetViewTarget!(self:Controller Object Reference=@n10) -> (ReturnValue:Actor Object Reference)
   ) -> (then)
   > n11 = StartListeningToStatefulEvent!(
    execute:Exec
    Target:Gameplay Event Router Component Object Reference=GetGlobalEventRouter()
    EventType:Script Struct Object Reference=&/Struct_LabradorRiding_WidgetProgress
    Context:Object Reference
    EventBubblingRule:EEventBubblingRule Enum
    None:%=GetOwningPlayer()
   ) -> (
    then:Exec
    AsyncTaskProxy:Async Action Start Listening To Stateful Event Object Reference
    OnEventReceived:Exec
    OnSavedEventStateExists:Exec
    OnEventStateCleared:Exec
    Payload:Struct Labrador Riding Widget Progress Structure
   ) {is=title K2Node_StartListeningToStatefulEventAsyncAction}
   n12 = Break Struct Labrador Riding Widget Progress!(Struct_LabradorRiding_WidgetProgress:"Struct Labrador Riding Widget Progress Structure (by ref)"=@n11.Payload) -> (Owner_5_#4:Actor Soft Object Reference, Visible_1_#5:bool=~) {is=title K2Node_BreakStruct}
   n13 = Cast To BP_AIPawn_Labrador_Ridable!(execute=@n11.OnEventReceived, Object=Resolve Soft Reference!(Input:Actor Soft Object Reference=@n12) -> (Output:Actor Object Reference)) -> (then, CastFailed, AsBP AIPawn Labrador Ridable:BP AIPawn Labrador Ridable Object Reference) {is=title K2Node_DynamicCast}
   > n14 = Set Labrador(Labrador:BP AIPawn Labrador Ridable Object Reference=@n13) -> (Output_Get:BP AIPawn Labrador Ridable Object Reference)
   > Bind Event to Training Increased(self:BP AIPawn Labrador Ridable Object Reference=@n14, Delegate=Create Event()) {is=title K2Node_AddDelegate}
   > n15 = Branch(Condition=@n12.Visible_1_#5)
   > SetHasActiveContent(bNewActive=true)
   > ShowProgress() {is=title K2Node_CallFunction}
   SetHasActiveContent(execute=@n15.else, bNewActive=false)
   > EndProgress() {is=title K2Node_CallFunction}
   OnViewTargetChanged!() -> (
    OutputDelegate
    then
    PlayerController:Fort Player Controller Object Reference
    OldViewTarget:Actor Object Reference
    NewViewTarget:Actor Object Reference
   ) {is=title K2Node_CustomEvent}
   Event On Become Active() {is=title K2Node_Event}
  ]}
 UnbindEvents {nodes [
   StopListeningToStatefulEvent(self:Async Action Start Listening To Stateful Event Object Reference)
   UnbindEvents() {is=title K2Node_FunctionEntry}
  ]}
]
variables [
 CachedPC {type=FortPlayerController, default=null}
 Handle_HideProgress {type=TimerHandle}
 Labrador {type=BP_AIPawn_Labrador_Ridable_C, default=null}
 Target {type=number}
]
widgets {
 info {inheritedWidgetCount=1, parentClass=/Script/FortniteUI.AthenaCommonPrioritizedWidget, rootWidgetClass=&SizeBox, widgetCount=14}
 widgets [
  SizeBox_2 {
   widgetClassPath=&SizeBox
   widget=SizeBox_2
   properties {bOverride_HeightOverride=true, bOverride_WidthOverride=true, heightOverride=120, widthOverride=650}
  }
  ScaleBox_0 {widgetClassPath=&ScaleBox, parent=SizeBox_2, widget=ScaleBox_0, slot {class=SizeBoxSlot}}
  Overlay_0 {
   widgetClassPath=&Overlay
   bIsVariable=true
   parent=ScaleBox_0
   widget=Overlay_0
   properties {renderTransformPivot=<0.5 0>}
   slot {class=ScaleBoxSlot}
  }
  TextContainer {
   widgetClassPath=&Overlay
   parent=Overlay_0
   widget=TextContainer
   properties {renderTransformPivot=<0.5 1>}
   slot {class=OverlaySlot, properties {horizontalAlignment=HAlign_Fill}}
  }
  Text {
   widgetClassPath=&CommonTextBlock
   parent=TextContainer
   widget=Text
   properties {
    bIsScrollingEnabled=false
    justification=Center
    style=/Game/Athena/HUD/TextStyles/TextStyle-HUD-MapScreen_Title.TextStyle-HUD-MapScreen_Title_C
    text=Llama Taming
    textTransformPolicy=ToUpper
    font {
     bForceMonospaced=false
     bMaterialIsStencil=false
     fontMaterial=null
     fontObject=&Burbank
     letterSpacing=0
     monospacedWidth=1
     size=18
     skewAmount=0
     typefaceFontName=BurbankRegular-Black
     outlineSettings {
      bApplyOutlineToDropShadows=true
      bMiteredCorners=false
      bSeparateFillAlpha=true
      outlineMaterial=&M-FontShadow
      outlineSize=3
      outlineColor=<rgba 0 0 0 0.5>
     }
    }
    renderTransform {angle=0, scale=<1 1>, shear=<-15 0>, translation=<0 0>}
    renderTransformPivot=<0.5 1>
    shadowOffset=<0 0>
   }
   slot {class=OverlaySlot, properties {horizontalAlignment=HAlign_Fill}}
  }
  BarContainer {
   widgetClassPath=&Overlay
   parent=Overlay_0
   widget=BarContainer
   slot {class=OverlaySlot, properties {verticalAlignment=VAlign_Center, padding=<ltrb 0 15 0 0>}}
  }
  BarSB {
   widgetClassPath=&SizeBox
   parent=BarContainer
   widget=BarSB
   properties {bOverride_HeightOverride=true, heightOverride=32}
   slot {class=OverlaySlot, properties {horizontalAlignment=HAlign_Fill, verticalAlignment=VAlign_Center}}
  }
  Overlay_5 {widgetClassPath=&Overlay, parent=BarSB, widget=Overlay_5, properties {bOverride_Cursor=true}, slot {class=SizeBoxSlot}}
  LeftBracket {
   widgetClassPath=&Image
   bIsVariable=true
   parent=Overlay_5
   widget=LeftBracket
   properties {brush {
     bIsDynamicallyLoaded=false
     drawAs=Image
     imageType=FullColor
     mirroring=NoMirror
     resourceName=None
     resourceObject=/LabradorRidingContent/Art/Materials/M_UI_ProgressBar_Bracket
     tiling=NoTile
     imageSize=<35 35>
     margin=<ltrb 0 0 0 0>
     outlineSettings {
      bUseBrushTransparency=false
      roundingType=HalfHeightRadius
      width=0
      color {colorUseRule=UseColor_Specified, specifiedColor=<rgba 0 0 0 0>}
      cornerRadii=<0 0 0 0>
     }
     tintColor {colorUseRule=UseColor_Specified, specifiedColor=<rgba 0.1301 0.7835 0.7991 1>}
     uVRegion {bIsValid=false, max=<0 0>, min=<0 0>}
    }}
   slot {class=OverlaySlot}
  }
  RightBracket {
   widgetClassPath=&Image
   bIsVariable=true
   parent=Overlay_5
   widget=RightBracket
   properties {
    brush {
     bIsDynamicallyLoaded=false
     drawAs=Image
     imageType=FullColor
     mirroring=NoMirror
     resourceName=None
     resourceObject=&M_UI_ProgressBar_Bracket
     tiling=NoTile
     imageSize=<35 35>
     margin=<ltrb 0 0 0 0>
     outlineSettings {
      bUseBrushTransparency=false
      roundingType=HalfHeightRadius
      width=0
      color {colorUseRule=UseColor_Specified, specifiedColor=<rgba 0 0 0 0>}
      cornerRadii=<0 0 0 0>
     }
     tintColor {colorUseRule=UseColor_Specified, specifiedColor=<rgba 0.1301 0.7835 0.7991 1>}
     uVRegion {bIsValid=false, max=<0 0>, min=<0 0>}
    }
    renderTransform {angle=0, scale=<-1 1>, shear=<0 0>, translation=<0 0>}
   }
   slot {class=OverlaySlot, properties {horizontalAlignment=HAlign_Right}}
  }
  Overlay_4 {widgetClassPath=&Overlay, parent=Overlay_5, widget=Overlay_4, slot {class=OverlaySlot, properties {padding=<ltrb 18 10 18 10>}}}
  ProgressBar {
   widgetClassPath=&Image
   bIsVariable=true
   parent=Overlay_4
   widget=ProgressBar
   properties {brush {
     bIsDynamicallyLoaded=false
     drawAs=Image
     imageType=FullColor
     mirroring=NoMirror
     resourceName=None
     resourceObject=/LabradorRidingContent/Art/Materials/M_UI_Status_Bar_Inst
     tiling=NoTile
     imageSize=<356 32>
     margin=<ltrb 0 0 0 0>
     outlineSettings {
      bUseBrushTransparency=false
      roundingType=HalfHeightRadius
      width=0
      color {colorUseRule=UseColor_Specified, specifiedColor=<rgba 0 0 0 0>}
      cornerRadii=<0 0 0 0>
     }
     tintColor {colorUseRule=UseColor_Specified, specifiedColor=<rgba 1 1 1 1>}
     uVRegion {bIsValid=false, max=<0 0>, min=<0 0>}
    }}
   slot {class=OverlaySlot, properties {horizontalAlignment=HAlign_Fill, verticalAlignment=VAlign_Fill}}
  }
  Detail {
   widgetClassPath=&Image
   bIsVariable=true
   parent=Overlay_4
   widget=Detail
   properties {
    visibility=Hidden
    brush {
     bIsDynamicallyLoaded=false
     drawAs=Image
     imageType=FullColor
     mirroring=NoMirror
     resourceName=None
     resourceObject=/LabradorRidingContent/Art/Textures/Group
     tiling=NoTile
     imageSize=<128 128>
     margin=<ltrb 0 0 0 0>
     outlineSettings {
      bUseBrushTransparency=false
      roundingType=HalfHeightRadius
      width=0
      color {colorUseRule=UseColor_Specified, specifiedColor=<rgba 0 0 0 0>}
      cornerRadii=<0 0 0 0>
     }
     tintColor {colorUseRule=UseColor_Specified, specifiedColor=<rgba 1 1 1 1>}
     uVRegion {bIsValid=false, max=<0 0>, min=<0 0>}
    }
    renderTransform {angle=0, scale=<1.25 2.6>, shear=<0 0>, translation=<338 -4>}
   }
   slot {class=OverlaySlot, properties {verticalAlignment=VAlign_Fill}}
  }
  TutorialHighlight {widgetClassPath=&TutorialHighlight}
 ]
}
